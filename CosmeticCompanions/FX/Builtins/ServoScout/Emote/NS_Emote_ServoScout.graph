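# Niagara system for the ServoScout cosmetic companion's emote: a timed, flickering green scanner projection.
# Cones_Glow_Flare_Light uses spawn-group switches to drive cone meshes, glow and flare sprites, and a light from one emitter.
# InnerLines and OuterLines swing mesh beams via per-partition angle offsets on a shared System.SystemQuat, while SmallScanLines spins short lines independently.
# System time-of-day values scale brightness, and a front-end flag raises cone opacity.
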NiagaraSystem /CosmeticCompanions/FX/Builtins/ServoScout/Emote/NS_Emote_ServoScout
dependencies {
 usedDataInterfaces [&NiagaraDataInterfaceCurve, &NiagaraDataInterfaceColorCurve, &NiagaraDataInterfaceArrayFloat]
 usedDynamicInputs [
  &ClampFloat
  &RemapRange
  &Subtract_Float
  /Niagara/DynamicInputs/Vectors/RotateVector
 ]
 usedModules [
  &self:SystemSpawnScript.SystemScriptSource.SystemScriptGraph.NiagaraNodeAssignment_6.SetVariables_#1
  &self:SystemSpawnScript.SystemScriptSource.SystemScriptGraph.NiagaraNodeAssignment_0.SetVariables_#2
  &self:SystemSpawnScript.SystemScriptSource.SystemScriptGraph.NiagaraNodeAssignment_1.SetVariables_#3
  &self:SystemSpawnScript.SystemScriptSource.SystemScriptGraph.NiagaraNodeAssignment_2.SetVariables_#4
  &self:SystemSpawnScript.SystemScriptSource.SystemScriptGraph.NiagaraNodeAssignment_3.SetVariables_#5
  &SystemState
  &self:SystemSpawnScript.SystemScriptSource.SystemScriptGraph.NiagaraNodeAssignment_4.SetVariables_#6
  &self:SystemSpawnScript.SystemScriptSource.SystemScriptGraph.NiagaraNodeAssignment_5.SetVariables_#7
  &self:SystemSpawnScript.SystemScriptSource.SystemScriptGraph.NiagaraNodeAssignment_7.SetVariables_#8
  &self:SystemSpawnScript.SystemScriptSource.SystemScriptGraph.NiagaraNodeAssignment_9.SetVariables_#9
 ]
}
emitters {
 Cones_Glow_Flare_Light {
  emitterSpawnScript [/SetVariables() {enabled=false, isSetParameters=true}]
  emitterUpdateScript [
   &EmitterState(Inactive Response=::NewEnumerator1, Life Cycle Mode=::NewEnumerator1, Loop Behavior=::NewEnumerator1, Loop Duration=7, UseLoopDelay=false)
   &SpawnBurst_Instantaneous(Spawn Count=1, Spawn Time=0.2, Use Spawn Probability=false)
   &SpawnBurst_Instantaneous(Spawn Count=1, Spawn Group=1, Spawn Time=0.2, Use Spawn Probability=false) {module=SpawnBurst_Instantaneous001}
   &SpawnBurst_Instantaneous(Spawn Group=2, Spawn Time=4.75, Use Spawn Probability=false, UseLoopCountLimit=false) {module=SpawnBurst_Instantaneous002}
   &SpawnBurst_Instantaneous(Spawn Group=3, Use Spawn Probability=false, UseLoopCountLimit=false) {module=SpawnBurst_Instantaneous003}
   &SpawnBurst_Instantaneous(Spawn Group=4, Use Spawn Probability=false, UseLoopCountLimit=false) {module=SpawnBurst_Instantaneous004}
   /SetVariables(Emitter.emitter_reveal:NiagaraFloat=$System.ScanWidthTime) {enabled=false, isSetParameters=true}
  ]
  particleSpawnScript [
   /SetVariables(Particles.VisibilityTag=&SpawnGroup()) {isSetParameters=true}
   &InitializeParticle(
    Color=&MergeHelper_ColorSwitch(
     0=&MakeCustomLinearColorFromBool(Bool=$User.bIsFrontEnd, True Color=<rgba 0.4325 1 0.0636 1.35>, False Color=<rgba 0.4325 1 0.0636 0.25>)
     1=<rgba 0.4325 1 0.0636 1.35>
     2=<rgba 0.2434 1 0.0009 0.4>
     3=<rgba 1 1 1 1>
     4=<rgba 0.2434 1 0.0009 0.2>
     Index=&SpawnGroup()
     Num=::NewEnumerator3
    )
    Lifetime=&MergeHelper_FloatSwitch(0=4.8, 1=4.8, 2=0.27, 3=5, 4=5.2, Index=&SpawnGroup(), Num=::NewEnumerator3)
    Mesh Scale=&MergeHelper_VectorSwitch(0=<0.02 0.1 1>, 1=<0.02 0.1 1>, 2=?, 3=?, Index=&SpawnGroup(), Num=::NewEnumerator0)
    Mesh Scale Mode=::NewEnumerator4
    Position Offset=&MergeHelper_VectorSwitch(0=<0 0 0>, 1=<2 0 0>, 2=<2 0 0>, 3=<0 0 0>, 4=<2.3 0 0>, Index=&SpawnGroup(), Num=::NewEnumerator3)
    Sprite Size Mode=::NewEnumerator3
    Uniform Sprite Size=&MergeHelper_FloatSwitch(0=0, 1=0, 2=100, 3=0, 4=70, Index=&SpawnGroup(), Num=::NewEnumerator3)
    UsePositionOffset=true
    Write Material Random=true
   )
   &InitialMeshOrientation(
    Mesh Axis To Orient=<0 0 0>
    Mesh Orientation Mode=::NewEnumerator2
    Orientation Coordinate Space=::Local
    Orientation Vector=&MergeHelper_VectorSwitch(0=<-1 0 0>, 1=<-1 0 0>, 2=?, 3=?, Index=&SpawnGroup(), Num=::NewEnumerator0)
    Rotation=<0 0 0.05>
    Rotation Coordinate Space=::NewEnumerator2
    Use Rotation Vector=true
   )
   &CameraOffset(Camera Offset Amount=&MergeHelper_FloatSwitch(0=0, 1=0, 2=15, 3=0, 4=15, Index=&SpawnGroup(), Num=::NewEnumerator3)) {module=CameraOffset001}
  ]
  particleUpdateScript [
   &ParticleState()
   &SolveForcesAndVelocity(Clamp Velocity=false, Limit Acceleration=false)
   &DynamicMaterialParameters(
    Index 0 Param 1=20
    Index 0 Param 2=0.1
    Index 1 Param 1=2
    Module.Param0WriteEnabled=true
    Module.Param0WriteEnabled1=true
    Module.Param1WriteEnabled=true
    Module.Param1WriteEnabled1=true
    Module.Param2WriteEnabled=true
    Module.Param2WriteEnabled1=true
    Module.Param3WriteEnabled=true
    Module.Param3WriteEnabled1=true
    Write Parameter Index 1=true
   )
   &ScaleColor(
    Scale Alpha=&MergeHelper_FloatSwitch(
     0=&Multiply_Float(
      A=&Multiply_Float(
       A=&FloatFromCurve(
        FloatCurve=DataInterface() {propertyValues {Curve {keys [
            <at 0 0 RCIM_Cubic>
            {arriveTangent=3, interpMode=RCIM_Cubic, leaveTangent=3, time=0.1, value=1}
            {arriveTangent=1.3097, interpMode=RCIM_Cubic, leaveTangent=1.3097, time=0.25, value=0.75}
            {arriveTangent=0.6, interpMode=RCIM_Cubic, leaveTangent=0.6, time=0.3672, value=1.35}
            <at 0.5 0.9 RCIM_Cubic>
            {arriveTangent=-0.6818, interpMode=RCIM_Cubic, leaveTangent=-0.6818, time=0.6221, value=1.35}
            {arriveTangent=-1.5293, interpMode=RCIM_Cubic, leaveTangent=-1.5293, time=0.72, value=0.75}
            {arriveTangent=-2.6786, interpMode=RCIM_Cubic, leaveTangent=-2.6786, time=0.9, value=0.925}
            <at 1 0 RCIM_Cubic>
           ]}}}
        CurveIndex=$Emitter.NormalizedLoopAge
       )
       B=&Waveform("Waveform [1]"=::NewEnumerator12, "[1] Frequency"=15, "[1] Amplitude Min/Max"=<0.5 0.9>)
      )
      B=&FNTimeOfDayScalar(Night Scale=0.5, Morning Scale=0.5, Day Scale=0.6, Evening Scale=0.4)
     )
     1=1
     2=&Multiply_Float(
      A=$System.ToD
      B=&FloatFromCurve(FloatCurve=DataInterface() {propertyValues {Curve {keys [
           {arriveTangent=0.8909, interpMode=RCIM_Cubic, leaveTangent=0.8909, tangentMode=RCTM_User, time=0, value=0}
           {arriveTangentWeight=0.0783, interpMode=RCIM_Cubic, leaveTangentWeight=0.0783, tangentMode=RCTM_User, time=0.15, value=1}
           {arriveTangentWeight=0.1044, interpMode=RCIM_Cubic, leaveTangentWeight=0.0522, tangentMode=RCTM_User, time=0.9, value=1}
           {arriveTangent=-1.2042, interpMode=RCIM_Cubic, leaveTangent=-1.2042, tangentMode=RCTM_User, time=1, value=0}
          ]}}})
     )
     3=1
     4=&Multiply_Float(
      A=$System.ToD
      B=&FloatFromCurve(FloatCurve=DataInterface() {propertyValues {Curve {keys [
           <at 0 0 RCIM_Cubic>
           {arriveTangentWeight=0.0783, interpMode=RCIM_Cubic, leaveTangentWeight=0.0783, tangentMode=RCTM_User, time=0.08, value=1}
           {arriveTangentWeight=0.1044, interpMode=RCIM_Cubic, leaveTangentWeight=0.0522, tangentMode=RCTM_User, time=0.92, value=1}
           {arriveTangent=-1.2042, interpMode=RCIM_Cubic, leaveTangent=-1.2042, tangentMode=RCTM_User, time=1, value=0}
          ]}}})
     )
     Index=&SpawnGroup()
     Num=::NewEnumerator3
    )
    ScaleA=true
    ScaleRGB=true
   )
   &ScaleColor(
    Color Value To Scale=$Particles.Color
    Scale Alpha=&MergeHelper_FloatSwitch(
     0=&FloatFromCurve(FloatCurve=DataInterface() {propertyValues {Curve {
         defaultValue=1
         keys [
          {interpMode=RCIM_Cubic, tangentMode=RCTM_User, time=0, value=0}
          {arriveTangent=8.8549, interpMode=RCIM_Cubic, leaveTangent=-0.1467, tangentMode=RCTM_Break, time=0.3, value=1}
          {arriveTangentWeight=0.3604, interpMode=RCIM_Cubic, leaveTangentWeight=0.0465, tangentMode=RCTM_User, time=0.92, value=1}
          <at 1 0 RCIM_Cubic>
         ]
        }}})
     1=&FloatFromCurve(FloatCurve=DataInterface() {propertyValues {Curve {
         defaultValue=1
         keys [
          {interpMode=RCIM_Cubic, tangentMode=RCTM_User, time=0, value=0}
          {arriveTangent=8.8549, interpMode=RCIM_Cubic, leaveTangent=-0.1467, tangentMode=RCTM_Break, time=0.3, value=1}
          {arriveTangentWeight=0.3604, interpMode=RCIM_Cubic, leaveTangentWeight=0.0465, tangentMode=RCTM_User, time=0.92, value=1}
          <at 1 0 RCIM_Cubic>
         ]
        }}})
     2=1
     3=?
     Index=&SpawnGroup()
     Num=::NewEnumerator1
    )
    Scale RGB=&MergeHelper_VectorSwitch(
     0=&VectorFromFloat(Value=&Lerp_Float(A=1, B=35, Alpha=$System.find_timeline))
     1=&VectorFromFloat(Value=&Lerp_Float(A=1, B=35, Alpha=$System.find_timeline))
     2=<1 1 1>
     3=?
     Index=&SpawnGroup()
     Num=::NewEnumerator1
    )
    ScaleA=true
    ScaleRGB=true
   ) {module=ScaleColor001}
   &ScaleSpriteSize(Scale Factor=&MergeHelper_Vector2Switch(
     0=<0 0>
     1=<0 0>
     2=&Vector2DFromFloat(Value=&FloatFromCurve(
       FloatCurve=DataInterface() {propertyValues {Curve {
          defaultValue=1.433
          keys [
           <at 0 0 RCIM_Cubic>
           {arriveTangentWeight=0.1667, interpMode=RCIM_Cubic, leaveTangentWeight=0.1667, tangentMode=RCTM_User, time=0.5, value=1}
           <at 1 0>
          ]
         }}}
       Scale Curve=1.2
      ))
     3=<0 0>
     4=&Vector2DFromFloat(Value=&FloatFromCurve(
       FloatCurve=DataInterface() {propertyValues {Curve {
          defaultValue=1.433
          keys [
           <at 0 0 RCIM_Cubic>
           {arriveTangentWeight=0.0745, interpMode=RCIM_Cubic, leaveTangentWeight=0.2981, tangentMode=RCTM_User, time=0.08, value=1}
           {arriveTangentWeight=0.1667, interpMode=RCIM_Cubic, leaveTangentWeight=0.1667, tangentMode=RCTM_User, time=0.95, value=1}
           <at 1 0>
          ]
         }}}
       Scale Curve=1.2
      ))
     Index=&SpawnGroup()
     Num=::NewEnumerator3
    ))
   &Light_Attributes(
    Color=1
    Diffuse Scale=1
    Enabled=1
    Light Diffuse Scale:NiagaraFloat=&FNTimeOfDayScalar(Night Scale=0.8, Morning Scale=0.65, Day Scale=0.8, Evening Scale=0.8)
    Light Position Offset:Vector3f=<30 0 0>
    Light Radius=110
    Light Specular Scale=0
    Position=1
    Position Offset Coordinate Space:ENiagaraCoordinateSpace=::Local
    Radius=1
    Scale Alpha:NiagaraFloat=&FloatFromCurve(FloatCurve=DataInterface() {propertyValues {Curve {
        defaultValue=1
        keys [
         {interpMode=RCIM_Cubic, tangentMode=RCTM_User, time=0, value=0}
         {interpMode=RCIM_Cubic, tangentMode=RCTM_User, time=0.2, value=5}
         {arriveTangentWeight=1.4434, interpMode=RCIM_Cubic, leaveTangentWeight=0.362, tangentMode=RCTM_User, time=0.8513, value=5.5447}
         <at 1 0 RCIM_Cubic>
        ]
       }}})
    Scale RGB:Vector3f=&MakeVectorFromLinearColorRGB(Color=<rgba 0.246 1 0 1>)
    ScaleA=true
    ScaleRGB=true
    Specular Scale=1
    UsePositionOffset=true
   )
   /SetVariables(Particles.MeshOrientation=$System.SystemQuat) {isSetParameters=true}
   /SetVariables(Particles.PlainConeColor:LinearColor=&Multiply_LinearColorByFloat(
     Linear Color=&ColorFromCurve(ColorCurve=DataInterface() {propertyValues {
        AlphaCurve {keys [<at 0 0>, <at 0.1074 0.2>, {arriveTangent=-0.2241, interpMode=RCIM_Cubic, leaveTangent=-0.2241, time=0.847, value=0.2}, <at 1 0 RCIM_Cubic>]}
        BlueCurve {keys [<at 0 0 RCIM_Cubic>]}
        GreenCurve {keys [<at 0 1 RCIM_Cubic>]}
        RedCurve {keys [<at 0 0.4844 RCIM_Cubic>]}
       }})
     Float=&FloatFromCurve(
      FloatCurve=DataInterface() {propertyValues {Curve {keys [
          <at 0 1 RCIM_Cubic>
          {arriveTangent=2.5, interpMode=RCIM_Cubic, leaveTangent=2.5, time=0.05, value=0.7}
          {arriveTangent=1, interpMode=RCIM_Cubic, leaveTangent=1, time=0.1, value=1.25}
          {arriveTangent=-0.75, interpMode=RCIM_Cubic, leaveTangent=-0.75, time=0.15, value=0.8}
          <at 0.2 1.175 RCIM_Cubic>
          {arriveTangent=1, interpMode=RCIM_Cubic, leaveTangent=1, time=0.25, value=0.8}
          {arriveTangent=-0.75, interpMode=RCIM_Cubic, leaveTangent=-0.75, time=0.3, value=1.275}
          {arriveTangent=0.5, interpMode=RCIM_Cubic, leaveTangent=0.5, time=0.35, value=0.725}
          {arriveTangent=0.5, interpMode=RCIM_Cubic, leaveTangent=0.5, time=0.4, value=1.325}
          {arriveTangent=-1, interpMode=RCIM_Cubic, leaveTangent=-1, time=0.45, value=0.775}
          {arriveTangent=-0.5, interpMode=RCIM_Cubic, leaveTangent=-0.5, time=0.5, value=1.225}
          {arriveTangent=0.25, interpMode=RCIM_Cubic, leaveTangent=0.25, time=0.55, value=0.725}
          {arriveTangent=-0.25, interpMode=RCIM_Cubic, leaveTangent=-0.25, time=0.6, value=1.25}
          <at 0.65 0.7 RCIM_Cubic>
          {arriveTangent=1.5, interpMode=RCIM_Cubic, leaveTangent=1.5, time=0.7, value=1.25}
          <at 0.75 0.85 RCIM_Cubic>
          <at 0.8 1.25 RCIM_Cubic>
          {arriveTangent=0.5, interpMode=RCIM_Cubic, leaveTangent=0.5, time=0.85, value=0.85}
          {arriveTangent=-1, interpMode=RCIM_Cubic, leaveTangent=-1, time=0.9, value=1.3}
          {arriveTangent=-3, interpMode=RCIM_Cubic, leaveTangent=-3, time=0.95, value=0.75}
          <at 1 1 RCIM_Cubic>
         ]}}}
      Scale Curve=&Multiply_Float(A=0.7, B=$System.ToD)
     )
    )) {isSetParameters=true}
   /SetVariables(Particles.Flash color:LinearColor=&MakeLinearColorFromVectorAndFloat(
     "Vector (RGB)"=&MakeVectorFromLinearColorRGB(Color=<rgba 0.4789 1 0.1949 1>)
     "Float (Alpha)"=&Multiply_Float(
      A=&Multiply_Float(
       A=1
       B=&FloatFromCurve(FloatCurve=DataInterface() {propertyValues {Curve {keys [<at 0 0 RCIM_Cubic>, <at 0.1 0.4>, {arriveTangent=-0.4444, interpMode=RCIM_Cubic, leaveTangent=-0.4444, time=0.9, value=0.4}, <at 1 0>]}}})
      )
      B=$System.ToD
     )
    )) {isSetParameters=true}
   /SetVariables(Particles.Flash size:Vector2f=&Multiply_Vector2D_ByFloat(
     Float=&FloatFromCurve(
      FloatCurve=DataInterface() {propertyValues {Curve {keys [
          <at 0 0 RCIM_Cubic>
          {arriveTangentWeight=0.0833, interpMode=RCIM_Cubic, leaveTangentWeight=0.3333, tangentMode=RCTM_User, time=0.5, value=1}
          <at 1 0>
         ]}}}
      Scale Curve=1.2
     )
     Vector2D=<60 40>
    )) {isSetParameters=true}
  ]
  rendererClasses [
   &NiagaraMeshRendererProperties
   &NiagaraSpriteRendererProperties
   &NiagaraLightRendererProperties
  ]
  renderers [
   {rendererClass=&NiagaraMeshRendererProperties, rendererIndex=0}
   {rendererClass=&NiagaraMeshRendererProperties, rendererIndex=1}
   {rendererClass=&NiagaraSpriteRendererProperties, rendererIndex=2}
   {rendererClass=&NiagaraSpriteRendererProperties, rendererIndex=3}
   {rendererClass=&NiagaraLightRendererProperties, rendererIndex=4}
   {rendererClass=&NiagaraSpriteRendererProperties, rendererIndex=5}
  ]
 }
 InnerLines {
  emitterSpawnScript [/SetVariables(Emitter.ScanWidth=100) {isSetParameters=true, module=SetVariables_#10}]
  emitterUpdateScript [
   &EmitterState()
   &SpawnBurst_Instantaneous(Spawn Count=4, Spawn Time=0.2, Use Spawn Probability=false, UseLoopCountLimit=false)
   /SetVariables(Emitter.ScanWidth=10) {isSetParameters=true, module=SetVariables_#11}
   &SpawnBurst_Instantaneous(Spawn Count=6, Spawn Time=5.9167, Use Spawn Probability=false, UseLoopCountLimit=false) {module=SpawnBurst_Instantaneous001}
   &SpawnBurst_Instantaneous(Spawn Count=4, Spawn Time=0.8833, Use Spawn Probability=false, UseLoopCountLimit=false) {module=SpawnBurst_Instantaneous002}
  ]
  particleSpawnScript [
   &InitializeParticle(
    Color Mode=::NewEnumerator0
    Lifetime=4.3
    Mesh Scale Mode=::NewEnumerator3
    Sprite Rotation Mode=::NewEnumerator2
    Sprite Size Mode=::NewEnumerator3
    Uniform Sprite Size=15
    UsePositionOffset=false
    Write Material Random=true
   )
   &PartitionParticles(Number Of Partitions=4, Particles Per Partition=1, Partition Mode:ENiagara_PartitionMode_v2=::NewEnumerator1)
   /SetVariables(Particles.Line_Normalized_Index:NiagaraFloat=&ReturnNormalizedExecIndex(Normalized Index Includes One=false)) {isSetParameters=true, module=SetVariables_#12}
   &Color()
   /SetVariables(Particles.CalculatedLineOrientation:Quat4f=&MakeQuatFromAxisAngle(
     Axis=&MakeVector(
      X=0
      Y=&Sine(Normalized Angle=$Particles.Line_Normalized_Index, Period=1, Scale=1, Bias=0)
      Z=&Cosine(Normalized Angle=$Particles.Line_Normalized_Index, Period=1, Scale=1, Bias=0)
     )
     AngleInDegrees=16
    )) {isSetParameters=true, module=SetVariables_#13}
   /SetVariables() {isSetParameters=true, module=SetVariables_#14}
   &InitialMeshOrientation(Mesh Orientation Mode=::NewEnumerator6, Orientation Quaternion:Quat4f=$Particles.CalculatedLineOrientation, Use Rotation Vector=false)
   /SetVariables(Particles.Line_Length:NiagaraFloat=&RandomRangeFloat(Minimum=50, Maximum=100)) {isSetParameters=true, module=SetVariables_#15}
  ]
  particleUpdateScript [
   &ParticleState()
   &DynamicMaterialParameters(
    Index 0 Param 1=5
    Module.Param0WriteEnabled=true
    Module.Param1WriteEnabled=true
    Module.Param2WriteEnabled=true
    Module.Param3WriteEnabled=true
   )
   /SetVariables(Particles.CalculatedLineOrientation:Quat4f=&MultiplyQuaternion(
     Quaternion A=$System.SystemQuat
     Quaternion B=&MultiplyQuaternion(
      Quaternion A=&MakeQuaternion(
       Quaternion From=0
       Axis=<0 0 1>
       Angle=&Add_Float(
        A=19
        B=&Sine(
         Normalized Angle=&Add_Float(
          A=$Engine.System.Age
          B=&Multiply_Float(A=0.25, B=&MakeFloatFromInt(Int=$Particles.PartitionParticles.Partition))
         )
         Period=1
         Scale=5
         Bias=0
        )
       )
       From=?
       To=?
       Stabilization Mode:ENiagara_OrientVectorCount=::NewEnumerator0
       Side Direction:Vector3f=?
       Up Direction:Vector3f=?
       Angle Type=::NewEnumerator0
       Normalize:NiagaraBool=true
      )
      Quaternion B=&MakeQuaternion(
       Quaternion From=0
       Axis=<0 0 1>
       Angle=&FloatFromCurve(
        FloatCurve=DataInterface() {propertyValues {Curve {keys [
            {arriveTangent=-0.7475, interpMode=RCIM_Cubic, leaveTangent=-0.7475, leaveTangentWeight=0.1167, tangentMode=RCTM_User, time=0, value=0}
            {arriveTangent=8.2015, tangentMode=RCTM_User, time=0.4, value=1}
            <at 4.7 1>
            <at 4.9 0>
           ]}}}
        CurveIndex=$Engine.System.Age
        Scale Curve=&SelectFloatFromArray(Direct Array Index=$Particles.PartitionParticles.Partition, Float Selection Array=DataInterface() {propertyValues {FloatData [4,-8,16,-19]}})
       )
       From=?
       To=?
       Stabilization Mode:ENiagara_OrientVectorCount=::NewEnumerator0
       Side Direction:Vector3f=?
       Up Direction:Vector3f=?
       Angle Type=::NewEnumerator0
       Coordinate Space=::Local
       Normalize:NiagaraBool=true
      )
     )
    )) {isSetParameters=true, module=SetVariables_#16}
   /SetVariables(Particles.Scale=&MakeVector(
     X=&Multiply_Float(
      A=$Particles.Line_Length
      B=&FloatFromCurve(
       FloatCurve=DataInterface() {propertyValues {Curve {
          defaultValue=0
          keys [
           {arriveTangent=8.9195, interpMode=RCIM_Cubic, leaveTangent=8.9195, tangentMode=RCTM_User, time=0, value=0}
           {interpMode=RCIM_Cubic, tangentMode=RCTM_User, time=0.08, value=1}
           {
            arriveTangentWeight=0.5508
            interpMode=RCIM_Cubic
            leaveTangent=-0.0015
            leaveTangentWeight=0.1132
            tangentMode=RCTM_Break
            time=0.92
            value=1
           }
           {arriveTangentWeight=0.0433, interpMode=RCIM_Cubic, tangentMode=RCTM_User, time=1, value=0}
          ]
         }}}
       Scale Curve=0.02
      )
     )
     Y=1
     Z=1
    )) {isSetParameters=true, module=SetVariables_#17}
   /SetVariables(Particles.dest_length:NiagaraFloat=&Multiply_Float(
     A=&Lerp_Float(
      A=90
      B=110
      Alpha=&Waveform(
       Global Phase=&Add_Float(A=$Emitter.NormalizedLoopAge, B=&Multiply_Float(A=$Particles.Line_Normalized_Index, B=20))
       "Waveform [1]"=::NewEnumerator12
       "[1] Phase"=&Add_Float(A=$Emitter.NormalizedLoopAge, B=&Multiply_Float(A=$Particles.Line_Normalized_Index, B=20))
       "[1] Frequency"=8
       "[1] Amplitude Min/Max"=<0 1>
      )
     )
     B=&FloatFromCurve(FloatCurve=DataInterface() {propertyValues {Curve {keys [
          {arriveTangent=-1.25, interpMode=RCIM_Cubic, time=0, value=0}
          {arriveTangent=-13.5437, interpMode=RCIM_Cubic, leaveTangent=-13.5437, tangentMode=RCTM_User, time=0.0552, value=1.4953}
          {arriveTangentWeight=0.058, interpMode=RCIM_Cubic, leaveTangentWeight=0.2402, tangentMode=RCTM_User, time=0.2, value=1}
          {arriveTangentWeight=0.3478, interpMode=RCIM_Cubic, leaveTangentWeight=0.0791, tangentMode=RCTM_User, time=0.95, value=1}
          {interpMode=RCIM_Cubic, leaveTangent=-1.1951, time=1, value=0}
         ]}}})
    )) {isSetParameters=true, module=SetVariables_#18}
   /SetVariables(Particles.SpritePos:NiagaraPosition=&RotateVectorByQuaternion(
     Vector To Rotate=&MakeVector(X=$Particles.dest_length, Y=0, Z=0)
     Quaternion=$Particles.CalculatedLineOrientation
    )) {isSetParameters=true, module=SetVariables_#19}
   /SetVariables(Particles.Line_Length:NiagaraFloat=&VectorLength(Vector=Linked() {linkedVariable {name=Particles.SpritePos, type=NiagaraPosition}})) {isSetParameters=true, module=SetVariables_#20}
   /SetVariables(Particles.MeshOrientation=$Particles.CalculatedLineOrientation) {isSetParameters=true, module=SetVariables_#21}
   &Color(Color=&Multiply_LinearColorByFloat(
     Linear Color=&ColorFromCurve(ColorCurve=DataInterface() {propertyValues {
        AlphaCurve {keys [<at 0 0.15 RCIM_Cubic>, {arriveTangent=-0.15, interpMode=RCIM_Cubic, leaveTangent=-0.15, time=0.9014, value=0.15}, <at 1 0 RCIM_Cubic>]}
        BlueCurve {keys [<at 0.0549 0.0815>]}
        GreenCurve {keys [<at 0.0549 1>]}
        RedCurve {keys [<at 0.0549 0.3002>]}
       }})
     Float=&Sine(
      Normalized Angle=&Add_Float(A=$Emitter.Age, B=&RandomRangeFloat(Minimum=0, Maximum=0.5))
      Period=&RandomRangeFloat(Minimum=0.65, Maximum=1, Evaluation Type=::NewEnumerator1)
      Scale=1
      Bias=&RandomRangeFloat(Minimum=0.7, Maximum=1, Evaluation Type=::NewEnumerator1)
     )
    )) {module=Color001}
   &SpriteRotationRate(Rotation Rate=50)
   &ScaleColor(
    Color Value To Scale=$Particles.Color
    Scale Alpha=&Multiply_Float(
     A=&FloatFromCurve(FloatCurve=DataInterface() {propertyValues {Curve {keys [<at 0.9 1 RCIM_Cubic>,<at 1 0>]}}}, Scale Curve=$System.ToD)
     B=&Waveform("Waveform [1]"=::NewEnumerator12, "[1] Frequency"=12, "[1] Amplitude Min/Max"=<0.5 1.1>)
    )
    ScaleA=true
    ScaleRGB=true
   )
  ]
  rendererClasses [&NiagaraMeshRendererProperties]
  renderers [{rendererClass=&NiagaraMeshRendererProperties, rendererIndex=0}]
 }
 OuterLines {
  emitterSpawnScript [/SetVariables(Emitter.ScanWidth=100) {isSetParameters=true, module=SetVariables_#10}]
  emitterUpdateScript [
   &EmitterState()
   &SpawnBurst_Instantaneous(Spawn Count=2, Use Spawn Probability=false, UseLoopCountLimit=false)
   /SetVariables(Emitter.ScanWidth=10) {isSetParameters=true, module=SetVariables_#11}
  ]
  particleSpawnScript [
   &InitializeParticle(
    Color Mode=::NewEnumerator0
    Lifetime=$Emitter.CurrentLoopDuration
    Mesh Scale Mode=::NewEnumerator3
    Sprite Rotation Mode=::NewEnumerator2
    Sprite Size Mode=::NewEnumerator3
    Uniform Sprite Size=15
    UsePositionOffset=false
    Write Material Random=true
   )
   &PartitionParticles(Particles Per Partition=1, Partition Mode:ENiagara_PartitionMode_v2=::NewEnumerator1)
   /SetVariables(Particles.Line_Normalized_Index:NiagaraFloat=&ReturnNormalizedExecIndex(Normalized Index Includes One=false)) {isSetParameters=true, module=SetVariables_#12}
   &Color()
   /SetVariables(Particles.CalculatedLineOrientation:Quat4f=&MakeQuatFromAxisAngle(
     Axis=&MakeVector(
      X=0
      Y=&Sine(Normalized Angle=$Particles.Line_Normalized_Index, Period=1, Scale=1, Bias=0)
      Z=&Cosine(Normalized Angle=$Particles.Line_Normalized_Index, Period=1, Scale=1, Bias=0)
     )
     AngleInDegrees=16
    )) {isSetParameters=true, module=SetVariables_#13}
   /SetVariables() {isSetParameters=true, module=SetVariables_#14}
   &InitialMeshOrientation(Mesh Orientation Mode=::NewEnumerator6, Orientation Quaternion:Quat4f=$Particles.CalculatedLineOrientation, Use Rotation Vector=false)
   /SetVariables(Particles.Line_Length:NiagaraFloat=&RandomRangeFloat(Minimum=50, Maximum=100)) {isSetParameters=true, module=SetVariables_#15}
  ]
  particleUpdateScript [
   &ParticleState()
   &DynamicMaterialParameters(
    Index 0 Param 1=5
    Module.Param0WriteEnabled=true
    Module.Param1WriteEnabled=true
    Module.Param2WriteEnabled=true
    Module.Param3WriteEnabled=true
   )
   /SetVariables(Particles.CalculatedLineOrientation:Quat4f=&MultiplyQuaternion(
     Quaternion A=$System.SystemQuat
     Quaternion B=&MakeQuaternion(
      Quaternion From=0
      Axis=<0 0 1>
      Angle=&FloatFromCurve(
       FloatCurve=DataInterface() {propertyValues {Curve {keys [
           <at 0 0 RCIM_Cubic>
           {arriveTangent=3.0863, tangentMode=RCTM_User, time=0.3, value=1}
           {arriveTangent=-4.2736, interpMode=RCIM_Cubic, leaveTangent=-4.2736, tangentMode=RCTM_User, time=4.5, value=1}
          ]}}}
       CurveIndex=$Engine.System.Age
       Scale Curve=&SelectFloatFromArray(Direct Array Index=$Particles.PartitionParticles.Partition, Float Selection Array=DataInterface() {propertyValues {FloatData [40,-5]}})
      )
      From=?
      To=?
      Stabilization Mode:ENiagara_OrientVectorCount=::NewEnumerator0
      Side Direction:Vector3f=?
      Up Direction:Vector3f=?
      Angle Type=::NewEnumerator0
      Coordinate Space=::Local
      Normalize:NiagaraBool=true
     )
    )) {isSetParameters=true, module=SetVariables_#16}
   /SetVariables(Particles.Scale=&MakeVector(
     X=&Multiply_Float(
      A=$Particles.Line_Length
      B=&FloatFromCurve(FloatCurve=DataInterface() {propertyValues {Curve {keys [<at 0.95 1>,<at 1 0>]}}}, Scale Curve=0.02)
     )
     Y=1
     Z=1
    )) {isSetParameters=true, module=SetVariables_#17}
   /SetVariables(Particles.dest_length:NiagaraFloat=&Multiply_Float(
     A=&Lerp_Float(
      A=90
      B=110
      Alpha=&Waveform(
       Global Phase=&Add_Float(A=$Emitter.NormalizedLoopAge, B=&Multiply_Float(A=$Particles.Line_Normalized_Index, B=20))
       "Waveform [1]"=::NewEnumerator12
       "[1] Phase"=&Add_Float(A=$Emitter.NormalizedLoopAge, B=&Multiply_Float(A=$Particles.Line_Normalized_Index, B=20))
       "[1] Frequency"=8
       "[1] Amplitude Min/Max"=<0 1>
      )
     )
     B=&FloatFromCurve(
      FloatCurve=DataInterface() {propertyValues {Curve {keys [
          {arriveTangent=-1.25, interpMode=RCIM_Cubic, time=0, value=0}
          {arriveTangent=-13.5437, interpMode=RCIM_Cubic, leaveTangent=-13.5437, tangentMode=RCTM_User, time=0.04, value=1.4953}
          {arriveTangent=-0.5695, interpMode=RCIM_Cubic, leaveTangent=-0.5695, time=0.1905, value=0.9925}
          {arriveTangent=-1.257, interpMode=RCIM_Cubic, leaveTangent=-1.257, time=0.9536, value=0.975}
          {interpMode=RCIM_Cubic, leaveTangent=-7.9488, time=0.98, value=0}
         ]}}}
      CurveIndex=$Emitter.NormalizedLoopAge
     )
    )) {isSetParameters=true, module=SetVariables_#18}
   /SetVariables(Particles.SpritePos:NiagaraPosition=&RotateVectorByQuaternion(
     Vector To Rotate=&MakeVector(X=$Particles.dest_length, Y=0, Z=0)
     Quaternion=$Particles.CalculatedLineOrientation
    )) {isSetParameters=true, module=SetVariables_#19}
   /SetVariables(Particles.Line_Length:NiagaraFloat=&VectorLength(Vector=Linked() {linkedVariable {name=Particles.SpritePos, type=NiagaraPosition}})) {isSetParameters=true, module=SetVariables_#20}
   /SetVariables(Particles.MeshOrientation=$Particles.CalculatedLineOrientation) {isSetParameters=true, module=SetVariables_#21}
   &Color(Color=&Multiply_LinearColorByFloat(
     Linear Color=&ColorFromCurve(ColorCurve=DataInterface() {propertyValues {
        AlphaCurve {keys [<at 0 0.15 RCIM_Cubic>, {arriveTangent=-0.15, interpMode=RCIM_Cubic, leaveTangent=-0.15, time=0.9014, value=0.15}, <at 1 0 RCIM_Cubic>]}
        BlueCurve {keys [<at 0.0549 0.0815>]}
        GreenCurve {keys [<at 0.0549 1>]}
        RedCurve {keys [<at 0.0549 0.3002>]}
       }})
     Float=&FloatFromCurve(FloatCurve=DataInterface() {propertyValues {Curve {keys [
          <at 0 1 RCIM_Cubic>
          {arriveTangent=-0.3852, interpMode=RCIM_Cubic, leaveTangent=-0.3852, time=0.1038, value=1.125}
          {arriveTangent=-0.3896, interpMode=RCIM_Cubic, leaveTangent=-0.3896, time=0.1947, value=0.925}
          {arriveTangent=0.1246, interpMode=RCIM_Cubic, leaveTangent=0.1246, time=0.2963, value=1.05}
          <at 0.3954 0.95 RCIM_Cubic>
          <at 0.4963 1.05 RCIM_Cubic>
          {arriveTangent=0.3648, interpMode=RCIM_Cubic, leaveTangent=0.3648, time=0.6034, value=0.95}
          <at 0.7019 1.125 RCIM_Cubic>
          {arriveTangent=-0.3643, interpMode=RCIM_Cubic, leaveTangent=-0.3643, time=0.7981, value=0.95}
          <at 0.9078 1.05 RCIM_Cubic>
          <at 1 0.95 RCIM_Cubic>
         ]}}})
    )) {module=Color001}
   &SpriteRotationRate(Rotation Rate=50)
   &ScaleColor(
    Color Value To Scale=$Particles.Color
    Scale Alpha=&Multiply_Float(
     A=&FloatFromCurve(
      FloatCurve=DataInterface() {propertyValues {Curve {keys [
          {arriveTangent=-10, interpMode=RCIM_Cubic, time=0, value=1}
          {arriveTangentWeight=0.4384, interpMode=RCIM_Cubic, leaveTangentWeight=0.033, tangentMode=RCTM_User, time=0.9, value=1}
          <at 1 0>
         ]}}}
      Scale Curve=$System.ToD
     )
     B=&Waveform("Waveform [1]"=::NewEnumerator12, "[1] Frequency"=12, "[1] Amplitude Min/Max"=<0.5 1.1>)
    )
    ScaleA=true
    ScaleRGB=true
   )
  ]
  rendererClasses [&NiagaraMeshRendererProperties]
  renderers [{rendererClass=&NiagaraMeshRendererProperties, rendererIndex=0}]
 }
 SmallScanLines {
  emitterSpawnScript [/SetVariables(Emitter.ScanWidth=100) {isSetParameters=true, module=SetVariables_#10}]
  emitterUpdateScript [
   &EmitterState()
   &SpawnBurst_Instantaneous(Spawn Count=3, Use Spawn Probability=false, UseLoopCountLimit=false)
   /SetVariables(Emitter.ScanWidth:NiagaraFloat=&Lerp_Float(A=35, B=55, Alpha=$System.ScanWidthTime)) {isSetParameters=true, module=SetVariables_#11}
  ]
  particleSpawnScript [
   &InitializeParticle(
    Color Mode=::NewEnumerator0
    Lifetime=$Emitter.CurrentLoopDuration
    Mesh Scale Mode=::NewEnumerator3
    Sprite Rotation Mode=::NewEnumerator2
    Sprite Size Mode=::NewEnumerator3
    Uniform Sprite Size=15
    UsePositionOffset=false
    Write Material Random=true
   )
   /SetVariables(Particles.Line_Normalized_Index:NiagaraFloat=&ReturnNormalizedExecIndex(Normalized Index Includes One=false)) {isSetParameters=true, module=SetVariables_#12}
   &Color()
   /SetVariables(Particles.CalculatedLineOrientation:Quat4f=&MakeQuatFromAxisAngle(
     Axis=&MakeVector(
      X=0
      Y=&Sine(Normalized Angle=$Particles.Line_Normalized_Index, Period=1, Scale=1, Bias=0)
      Z=&Cosine(Normalized Angle=$Particles.Line_Normalized_Index, Period=1, Scale=1, Bias=0)
     )
     AngleInDegrees=16
    )) {isSetParameters=true, module=SetVariables_#13}
   /SetVariables() {isSetParameters=true, module=SetVariables_#14}
   &InitialMeshOrientation(Mesh Orientation Mode=::NewEnumerator6, Orientation Quaternion:Quat4f=$Particles.CalculatedLineOrientation, Use Rotation Vector=false)
   /SetVariables(Particles.Line_Length:NiagaraFloat=&RandomRangeFloat(Minimum=50, Maximum=150)) {isSetParameters=true, module=SetVariables_#15}
  ]
  particleUpdateScript [
   &ParticleState()
   &DynamicMaterialParameters(
    Index 0 Param 1=5
    Module.Param0WriteEnabled=true
    Module.Param1WriteEnabled=true
    Module.Param2WriteEnabled=true
    Module.Param3WriteEnabled=true
   )
   /SetVariables(Particles.CalculatedLineOrientation:Quat4f=&MultiplyQuaternion(
     Quaternion A=<0 0 0 1>
     Quaternion B=&MultiplyQuaternion(
      Quaternion A=&MakeQuatFromAxisAngle(
       Axis=<1 0 0>
       AngleInDegrees=&Multiply_Float(
        A=&Add_Float(A=$Emitter.NormalizedLoopAge, B=&RandomRangeFloat(Minimum=-0.2, Maximum=0.2))
        B=&RandomRangeFloat(Minimum=900, Maximum=-1080)
       )
      )
      Quaternion B=&MultiplyQuaternion(
       Quaternion A=&MakeQuatFromAxisAngle(
        Axis=&MakeVector(
         X=1
         Y=&Sine(Normalized Angle=$Particles.Line_Normalized_Index, Period=1, Scale=1, Bias=0)
         Z=&Cosine(Normalized Angle=$Particles.Line_Normalized_Index, Period=1, Scale=1, Bias=0)
        )
        AngleInDegrees=$Emitter.ScanWidth
       )
       Quaternion B=<0 0 0 1>
      )
     )
    )) {isSetParameters=true, module=SetVariables_#16}
   /SetVariables(Particles.Scale=&MakeVector(X=&Multiply_Float(A=$Particles.Line_Length, B=0.02), Y=1, Z=1)) {isSetParameters=true, module=SetVariables_#17}
   /SetVariables(Particles.dest_length:NiagaraFloat=&Multiply_Float(
     A=&Lerp_Float(
      A=5
      B=10
      Alpha=&Waveform(
       Global Phase=&Add_Float(A=$Emitter.NormalizedLoopAge, B=&Multiply_Float(A=$Particles.Line_Normalized_Index, B=20))
       "Waveform [1]"=::NewEnumerator12
       "[1] Phase"=&Add_Float(A=$Emitter.NormalizedLoopAge, B=&Multiply_Float(A=$Particles.Line_Normalized_Index, B=20))
       "[1] Frequency"=8
       "[1] Amplitude Min/Max"=<0 1>
      )
     )
     B=&FloatFromCurve(
      FloatCurve=DataInterface() {propertyValues {Curve {keys [
          {arriveTangent=-1.25, interpMode=RCIM_Cubic, time=0, value=0}
          {arriveTangent=-13.5437, interpMode=RCIM_Cubic, leaveTangent=-13.5437, tangentMode=RCTM_User, time=0.0552, value=1.4953}
          {arriveTangent=-0.5791, interpMode=RCIM_Cubic, leaveTangent=-0.5791, time=0.1905, value=0.9925}
          {arriveTangent=-1.257, interpMode=RCIM_Cubic, leaveTangent=-1.257, time=0.9536, value=0.975}
          {interpMode=RCIM_Cubic, leaveTangent=-7.9488, time=0.98, value=0}
         ]}}}
      CurveIndex=$Emitter.NormalizedLoopAge
     )
    )) {isSetParameters=true, module=SetVariables_#18}
   /SetVariables(Particles.SpritePos:NiagaraPosition=&RotateVectorByQuaternion(
     Vector To Rotate=&MakeVector(X=$Particles.dest_length, Y=0, Z=0)
     Quaternion=$Particles.CalculatedLineOrientation
    )) {isSetParameters=true, module=SetVariables_#19}
   /SetVariables(Particles.Line_Length:NiagaraFloat=&VectorLength(Vector=Linked() {linkedVariable {name=Particles.SpritePos, type=NiagaraPosition}})) {isSetParameters=true, module=SetVariables_#20}
   /SetVariables(Particles.MeshOrientation=$Particles.CalculatedLineOrientation) {isSetParameters=true, module=SetVariables_#21}
   &Color(Color=&ColorFromCurve(
     ColorCurve=DataInterface() {propertyValues {
       AlphaCurve {keys [
         <at 0 0 RCIM_Cubic>
         {arriveTangent=0.4129, interpMode=RCIM_Cubic, leaveTangent=0.4129, time=0.0287, value=0.4}
         {arriveTangent=-0.4167, interpMode=RCIM_Cubic, leaveTangent=-0.4167, time=0.9688, value=0.4}
         <at 0.9886 0 RCIM_Cubic>
        ]}
       BlueCurve {keys [<at 0.0549 0.0896>]}
       GreenCurve {keys [<at 0.0549 1.1>]}
       RedCurve {keys [<at 0.0549 0.3302>]}
      }}
     CurveIndex=$Emitter.NormalizedLoopAge
    )) {module=Color001}
   &SpriteRotationRate(Rotation Rate=50)
   &ScaleColor(
    Color Value To Scale=$Particles.Color
    Scale Alpha=&Multiply_Float(
     A=&FloatFromCurve(
      FloatCurve=DataInterface() {propertyValues {Curve {keys [<at 0.7 0.15 RCIM_Cubic>,<at 1 0>]}}}
      Scale Curve=&Multiply_Float(A=5, B=$System.ToD)
     )
     B=&Waveform("Waveform [1]"=::NewEnumerator12, "[1] Frequency"=12, "[1] Amplitude Min/Max"=<0.5 1.1>)
    )
    ScaleA=true
    ScaleRGB=true
   )
  ]
  rendererClasses [&NiagaraMeshRendererProperties]
  renderers [{rendererClass=&NiagaraMeshRendererProperties, rendererIndex=0}]
 }
}
summary {
 systemName=NS_Emote_ServoScout
 emitters [
  {emitterName=SmallScanLines, rendererClasses [&NiagaraMeshRendererProperties]}
  {emitterName=OuterLines, rendererClasses [&NiagaraMeshRendererProperties]}
  {emitterName=InnerLines, rendererClasses [&NiagaraMeshRendererProperties]}
  {
   emitterName=Cones_Glow_Flare_Light
   rendererClasses [
    &NiagaraMeshRendererProperties
    &NiagaraSpriteRendererProperties
    &NiagaraLightRendererProperties
   ]
  }
 ]
 userVariables [
  User.bIsFrontEnd {type=NiagaraBool}
  User.bIsFrontEndPreview {type=NiagaraBool}
  User.bisHeadHidden {type=NiagaraBool}
  User.IsLargeBody {type=NiagaraBool, defaultValue=true}
  User.LaserColorArray {
   type=NiagaraDataInterfaceArrayColor
   description=""
   defaultValue {
    struct=NiagaraExt_VariableValue_DataInterface
    value {
     dataInterface=&self:NiagaraDataInterfaceArrayColor_0
     dataInterfaceClass=&NiagaraDataInterfaceArrayColor
    }
   }
  }
  User.LasersPositionArray {
   type=NiagaraDataInterfaceArrayFloat3
   description=""
   defaultValue {
    struct=NiagaraExt_VariableValue_DataInterface
    value {
     dataInterface=&self:NiagaraDataInterfaceArrayFloat3_0
     dataInterfaceClass=&NiagaraDataInterfaceArrayFloat3
    }
   }
  }
  User.LaserState {type=NiagaraFloat, defaultValue=0, description=""}
  User.OwnerSkeleton {
   type=NiagaraDataInterfaceSkeletalMesh
   defaultValue {
    struct=NiagaraExt_VariableValue_DataInterface
    value {dataInterface=&self:NiagaraDataInterfaceSkeletalMesh_0}
   }
  }
 ]
}
