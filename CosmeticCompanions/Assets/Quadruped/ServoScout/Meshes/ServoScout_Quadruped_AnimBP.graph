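# Animation Blueprint for the ServoScout quadruped cosmetic companion that retargets its pose from another mesh and drives an emote effect.
# The AnimGraph outputs Retarget Pose From Mesh through RTG_ServoScout, leaving a Copy Pose From Mesh node unconnected.
# Each update, the Niagara function reads the NS_Active curve, spawning the emote system on the jaw when it exceeds a threshold and deactivating it otherwise.
# A boolean guard prevents repeated spawns.
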AnimBlueprint /CosmeticCompanions/Assets/Quadruped/ServoScout/Meshes/ServoScout_Quadruped_AnimBP
target_skeleton=&/ServoScout_Quadruped_Skeleton
graphs [
 AnimGraph {nodes [
   Copy Pose From Mesh() {bCopyCurves=true, bUseAttachedParent=true}
   Output Pose(Result=Retarget Pose From Mesh!() -> (Pose:Pose Link Structure) {is=title AnimGraphNode_RetargetPoseFromMesh, iKRetargeterAsset=&/RTG_ServoScout})
  ]}
 EventGraph {nodes [
   Event BlueprintUpdateAnimation()
   > Niagara()
   TryGetPawnOwner()
  ]}
 Niagara {nodes [
   Niagara() {is=title K2Node_FunctionEntry}
   > n1 = Branch(Condition="float > float"(A=GetCurveValue(CurveName=NS_Active), B=0.1))
   > n2 = Branch(Condition=Get bNS_Spawned() -> (bNS_Spawned:bool))
   n3 = SpawnSystemAttached!(
    execute=@n2.else
    SystemTemplate:Niagara System Object Reference=/CosmeticCompanions/FX/Builtins/ServoScout/Emote/NS_Emote_ServoScout
    AttachToComponent:Scene Component Object Reference=GetOwningComponent()
    AttachPointName:Name=jaw
    Location:Vector="2.5,7,0"
    Rotation:Rotator="0,90,0"
    LocationType:EAttachLocation Enum=SnapToTarget
    bAutoDestroy:bool=true
    bAutoActivate:bool=true
    PoolingMethod:ENCPoolMethod Enum=None
    bPreCullCheck:bool=true
   ) -> (then, ReturnValue:Niagara Particle System Component Object Reference)
   > Set SpawnedNS(SpawnedNS:Niagara Particle System Component Object Reference=@n3) -> (Output_Get:Niagara Particle System Component Object Reference)
   > Set bNS_Spawned(bNS_Spawned:bool=true) -> (Output_Get:bool)
   Branch(execute=@n1.else, Condition=Get bNS_Spawned())
   > Deactivate(self:Actor Component Object Reference=Get SpawnedNS() -> (SpawnedNS:Niagara Particle System Component Object Reference))
   > Set bNS_Spawned(bNS_Spawned=false)
  ]}
]
variables [
 bNS_Spawned {type=boolean}
 SpawnedNS {type=NiagaraComponent, default=null}
]
